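# Editor utility Blueprint whose ReparentToSDP function bulk-reparents selected Blueprint assets to the class held in L_NewClass.
# It loops over the selection, gets each Blueprint asset and compares its parent class to the target.
# When they differ, it checks out the asset and calls ReparentBlueprint.
# A Python snippet for reloading the selected packages hangs off a Sequence with no incoming execution link, so it does not run.
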EditorUtilityBlueprint /Game/ArtTools/Environment/BulkReparent/AAU_ReparentToSDP
graphs [
 EventGraph {links [], nodes []}
 ReparentToSDP {nodes [
   n1 = Sequence()
   n2 = GetSelectedAssets(execute=@n3.then)
   Execute Python Script!(
    execute=@n1.then_1
    PythonScript="import unreal\r\npackages = []\r\n\r\nfor asset in assets:\r\n    asset_package = asset.get_package()\r\n    packages.append(asset_package)\r\n\r\nunreal.EditorLoadingAndSavingUtils.reload_packages(packages, unreal.ReloadPackagesInteractionMode.ASSUME_POSITIVE)\r\n"
    assets:Array of Object References=@n2
   ) -> (then, ReturnValue) {inputs [assets]}
   n3 = ReparentToSDP() {is=title K2Node_FunctionEntry}
   n4 = For Each Loop(Exec=@n2.then, Array:Array of Object References=@n2) -> (Array Element:Object Reference)
   > n5 = GetBlueprintAsset(Object=@n4)
   ReparentBlueprint(execute=@n7.then, Blueprint=@n5, NewParentClass=Get L_NewClass() -> (L_NewClass:Object Class Reference))
   n6 = ParentClass!(execute=@n5.then, BlueprintObj:Blueprint Object Reference=@n5) -> (then, ReturnValue:Object Class Reference)
   > Branch(Condition="Not Equal (Class)"(A:Object Class Reference=@n6, B:Object Class Reference=Get L_NewClass()) -> (ReturnValue:bool=~) {is=title K2Node_PromotableOperator})
   > n7 = CheckoutLoadedAsset!(execute, AssetToCheckout:Object Reference=@n5) -> (then, ReturnValue:bool)
  ]}
]
variables []
